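# Niagara idle effect that draws wing motion blur on the HoneySting companion pet.
# RibbonBlur spawns short-lived, faint anim-trail ribbons between bone or socket pairs.
# RopeBlur keeps two mesh particles on sockets, alternating a sampled socket by tick parity to derive width, radius and a CalcRopeAngle-driven material parameter, with orientation relative to a reference socket.
# User.WingFlapAlpha, time-of-day and front-end scalars modulate both emitters' intensity.
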NiagaraSystem /CosmeticCompanions/FX/Builtins/HoneySting/Idle/NS_Pet_HoneySting_WingBlur
dependencies {
 usedDataInterfaces [
  &NiagaraDataInterfaceCurve
  &NiagaraDataInterfaceSocketReader
  &NiagaraDataInterfaceColorCurve
  &NiagaraDataInterfacePlatformSet
 ]
 usedDynamicInputs [&FN_PickFloatDependOnPlatform]
 usedModules [
  &self:SystemSpawnScript.SystemScriptSource.SystemScriptGraph.NiagaraNodeAssignment_2.SetVariables_#1
  &self:SystemSpawnScript.SystemScriptSource.SystemScriptGraph.NiagaraNodeAssignment_4.SetVariables_#2
  &SystemState
  &self:SystemSpawnScript.SystemScriptSource.SystemScriptGraph.NiagaraNodeAssignment_0.SetVariables_#3
  &self:SystemSpawnScript.SystemScriptSource.SystemScriptGraph.NiagaraNodeAssignment_1.SetVariables_#4
 ]
}
emitters {
 RibbonBlur {
  emitterSpawnScript [&FN_RibbonIDOffsetUtility()]
  emitterUpdateScript [
   &EmitterState(
    Inactive Response=::NewEnumerator1
    Life Cycle Mode=::NewEnumerator1
    Loop Behavior=::NewEnumerator1
    Loop Duration Mode=::NewEnumerator1
    UseLoopDelay=false
   )
   &FN_SpawnAnimTrail(Number of Anim Trails=6)
  ]
  particleSpawnScript [
   &FN_InitializeAnimTrailBoneIndecies(First Bone or Socket Index=10, Second Bone or Socket Index=0)
   &FN_InitializeAnimTrailBoneIndecies(AnimTrail Selection=1, First Bone or Socket Index=1, Second Bone or Socket Index=10) {module=FN_InitializeAnimTrailBoneIndecies004}
   &FN_InitializeAnimTrailBoneIndecies(AnimTrail Selection=2, First Bone or Socket Index=6, Second Bone or Socket Index=7) {module=FN_InitializeAnimTrailBoneIndecies003}
   &FN_InitializeAnimTrailBoneIndecies(AnimTrail Selection=3, First Bone or Socket Index=3, Second Bone or Socket Index=11) {module=FN_InitializeAnimTrailBoneIndecies002}
   &FN_InitializeAnimTrailBoneIndecies(AnimTrail Selection=3, First Bone or Socket Index=4, Second Bone or Socket Index=11) {module=FN_InitializeAnimTrailBoneIndecies005}
   &FN_InitializeAnimTrailBoneIndecies(AnimTrail Selection=4, First Bone or Socket Index=9, Second Bone or Socket Index=8) {module=FN_InitializeAnimTrailBoneIndecies001}
   &FN_InitializeAnimTrail(Color=<rgba 1 1 1 0.13>, Lifetime=0.15)
  ]
  particleUpdateScript [
   &FN_UpdateAnimTrail() {module=FN_UpdateAnimTrail001}
   &DynamicMaterialParameters(
    Index 0 Param 1=1
    Index 0 Param 4=1
    Index 1 Param 4=1
    Index 2 Param 1=10000
    Index 2 Param 2=2
    Index 2 Param 4=0.6
    Module.Index0Float0Write=true
    Module.Index0Float1Write=true
    Module.Index0Float2Write=true
    Module.Index0Float3Write=true
    Module.Index1Float0Write=true
    Module.Index1Float1Write=true
    Module.Index1Float2Write=true
    Module.Index1Float3Write=true
    Module.Index2Float0Write=true
    Module.Index2Float1Write=true
    Module.Index2Float2Write=false
    Module.Index2Float3Write=false
    Write Parameter Index 1=::NewEnumerator1
    Write Parameter Index 2=::NewEnumerator1
   )
   &ScaleColor(
    Scale Alpha=&FloatFromCurve(
     FloatCurve=DataInterface() {propertyValues {Curve {keys [
         {arriveTangent=-2.0607, interpMode=RCIM_Cubic, leaveTangent=-2.0607, tangentMode=RCTM_User, time=0, value=1}
         {interpMode=RCIM_Cubic, leaveTangent=-2.0607, time=1, value=0}
        ]}}}
     Scale Curve=&Multiply_Float(A=&Multiply_Float(A=$System.TODScalar, B=$System.FEScalar), B=$User.WingFlapAlpha)
    )
    ScaleA=true
    ScaleRGB=true
   )
  ]
  rendererClasses [&NiagaraRibbonRendererProperties]
  renderers [{rendererClass=&NiagaraRibbonRendererProperties}]
 }
 RopeBlur {
  emitterUpdateScript [
   &EmitterState(
    Inactive Response=::NewEnumerator1
    Life Cycle Mode=::NewEnumerator1
    Loop Behavior=::NewEnumerator1
    Loop Duration Mode=::NewEnumerator1
    UseLoopDelay=false
   )
   &SpawnBurst_Instantaneous(Spawn Count=2, Use Spawn Probability=false, UseLoopCountLimit=false)
  ]
  particleSpawnScript [
   &PartitionParticles()
   &InitializeParticle(Color=<rgba 1.8 2.54 3 0.1>, Lifetime=5, Mesh Scale Mode=::NewEnumerator3, UsePositionOffset=false, Write Material Random=true)
   /SetVariables() {isSetParameters=true}
   &InitialMeshOrientation(Mesh Orientation Mode=::NewEnumerator7, Use Rotation Vector=true)
  ]
  particleUpdateScript [
   &ParticleState(Kill Particles When Lifetime Has Elapsed=false)
   &SocketLocation(
    Orientation Sampling=::NewEnumerator1
    Position Sampling=::NewEnumerator1
    SamplePosOffset=false
    Socket Index=12
    Socket Reader=$System.Socket Reader
    Socket Sampling Mode=::NewEnumerator3
    VelocitySpeedLimit=false
   ) {module=SocketLocation002}
   &SocketLocation(
    SamplePosOffset=false
    Sampled Position Coordinate Space:ENiagaraCoordinateSpace=::Local
    Sampled Velocity Coordinate Space:ENiagaraCoordinateSpace=::Local
    Socket Index=&MergeHelper_IntSwitch(0=0, 1=3, 2=?, 3=?, 4=?, 5=?, Index=$Particles.PartitionParticles.Partition, Num=::NewEnumerator0)
    Socket Reader=$System.Socket Reader
    Socket Sampling Mode=::NewEnumerator3
    VelocitySpeedLimit=false
   )
   &SocketLocation(
    Orientation Sampling=::NewEnumerator1
    Position Sampling=::NewEnumerator1
    SamplePosOffset=false
    Sampled Orientation Coordinate Space:ENiagaraCoordinateSpace=::Local
    Sampled Position Coordinate Space:ENiagaraCoordinateSpace=::Local
    Sampled Velocity Coordinate Space:ENiagaraCoordinateSpace=::Local
    Socket Index=&Add_Integer(
     A=&ModuloInt(A=$Engine.System.TickCount, B=2)
     B=&MergeHelper_IntSwitch(0=1, 1=4, 2=?, 3=?, 4=?, 5=?, Index=$Particles.PartitionParticles.Partition, Num=::NewEnumerator0)
    )
    Socket Reader=$System.Socket Reader
    Socket Sampling Mode=::NewEnumerator3
    VelocitySpeedLimit=false
   ) {module=SocketLocation001}
   /SetVariables(Particles.MeshOrientation=&MultiplyQuaternion(
     Quaternion A=&MultiplyQuaternion(
      Quaternion A=/Niagara/DynamicInputs/Rotation/InvertQuaternion(Quat:Quat4f=$Output.SocketLocation002.SampledRotation)
      Quaternion B=$Output.SocketLocation.SampledRotation
     )
     Quaternion B=$Particles.Initial.MeshOrientation
    )) {isSetParameters=true}
   &self:CalcRopeAngle(SampledPosition:NiagaraPosition=$Output.SocketLocation.SampledPosition) {module=CalcRopeAngle}
   /SetVariables(
    Particles.RopeRadius:NiagaraFloat=&Divide_Float(
     A=&VectorLength(Vector=&Lerp_Vector(
       A=Linked() {linkedVariable {name=Output.SocketLocation.SampledPosition, type=NiagaraPosition}}
       B=Linked() {linkedVariable {name=Output.SocketLocation001.SampledPosition, type=NiagaraPosition}}
       Alpha=0.5
      ))
     B=95
    )
    Particles.RopeWidth:NiagaraFloat=&Divide_Float(
     A=&DistanceBetweenPositions(A=$Output.SocketLocation001.SampledPosition, B=Linked() {linkedVariable {name=Particles.RopeWidthPreviousPosition, type=Vector3f}})
     B=70
    )
   ) {isSetParameters=true}
   &ScaleMeshSize(Scale Factor=&MakeVector(X=$Particles.RopeWidth, Y=$Particles.RopeRadius, Z=$Particles.RopeRadius))
   &DynamicMaterialParameters(
    Index 0 Param 1=&Subtract_Float(
     A=$Particles.CalcRopeAngle.RopeAngle
     B=&MergeHelper_FloatSwitch(
      0=&FN_FrontEndFloatModifier(Gameplay=0.12, Front End=0.095)
      1=&Add_Float(A=1.5, B=&FN_FrontEndFloatModifier(Gameplay=0.12, Front End=0.095))
      2=?
      3=?
      Index=$Particles.PartitionParticles.Partition
      Num=::NewEnumerator0
     )
    )
    Index 0 Param 2=0.3
    Module.Index0Float0Write=true
    Module.Index0Float1Write=true
    Module.Index0Float2Write=false
    Module.Index0Float3Write=false
   )
   &ScaleColor(
    Scale Alpha=&ClampFloat(Float=&Subtract_Float(A=&Multiply_Float(A=$Engine.System.Age, B=2), B=0))
    Scale RGB=&VectorFromFloat(Value=&Multiply_Float(
      A=&Multiply_Float(
       A=&FNTimeOfDayScalar(Night Scale=0.5, Morning Scale=0.8, Evening Scale=0.7)
       B=&Multiply_Float(A=$System.FEScalar, B=0.8)
      )
      B=$User.WingFlapAlpha
     ))
    ScaleA=true
    ScaleRGB=true
   )
   /SetVariables(Particles.RopeWidthPreviousPosition:NiagaraPosition=$Output.SocketLocation001.SampledPosition) {isSetParameters=true}
  ]
  rendererClasses [&NiagaraMeshRendererProperties]
  renderers [{rendererClass=&NiagaraMeshRendererProperties, rendererIndex=0}]
 }
}
summary {
 systemName=NS_Pet_HoneySting_WingBlur
 emitters [
  {emitterName=RopeBlur, rendererClasses [&NiagaraMeshRendererProperties]}
  {emitterName=RibbonBlur, rendererClasses [&NiagaraRibbonRendererProperties]}
 ]
 userVariables [
  User.bIsFrontEnd {type=NiagaraBool}
  User.bIsFrontEndPreview {type=NiagaraBool}
  User.RibbonIDOffset {type=NiagaraInt32}
  User.WingFlapAlpha {type=NiagaraFloat, defaultValue=1, description=""}
 ]
}
